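; Passive player augment ability that applies a reload speed effect to the owner while the current weapon's magazine is empty.
; On authority, activation waits for weapon fire events and binds to the pawn's weapon-equipped delegate.
; Both triggers re-check the magazine, applying the GameplayEffect and storing its handle when empty, otherwise removing it by that handle.
; No ammo-type filter or existing-effect check precedes application, and debug prints remain.
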(event Ability|EventFailedToActivatePassiveAbility
  (params (ActorInfo :type Gameplay%20Ability%20Actor%20Info%20Structure)))

(event Ability|EventActivateAbility
  (Ability|HasAuthority
    (:True
      (Ability|Tasks|WaitGameplayEvent "(TagName=\"Event.Weapons.Fire\")"
        (:then
          (bind _outfortplayerpawn (CallFunction|GetAvatarasValidFortPlayerPawn)
            (:ValidFortPlayerPawn
              (FortPawn|BindEventtoOnWeaponEquippedDelegate _outfortplayerpawn (delegate Custom|WeaponChange)))))
        (:EventReceived
          (CallFunction|ApplyorRemoveGE))))))

(event Custom|WeaponChange
  (params (NewWeapon :type Fort%20Weapon%20Object%20Reference) (PrevWeapon :type Fort%20Weapon%20Object%20Reference))
  (CallFunction|ApplyorRemoveGE))

(event Custom|ApplyOrRemoveGE
  (bind _bisempty (CallFunction|IsWeaponAmmoEmpty))
  (if _bisempty
    (bind _returnvalue (Ability|ApplyGameplayEffectToOwner "/ChronoPlayerAugments/Items/Augments/Ch4S3/MediumEmptyReload/GE_PA_MediumEmptyReload_ReloadSpeed.GE_PA_MediumEmptyReload_ReloadSpeed_C" 0))
    (Variables|Default|SetReloadSpeedGE _returnvalue)
    (Development|PrintString "apply GE")
    (else
      (CallFunction|RemoveGE))))

(event Custom|RemoveGE
  (Ability|RemoveGameplayEffectFromOwnerWithHandle (Variables|Default|GetReloadSpeedGE))
  (Development|PrintString "remove GE"))

(fn IsWeaponAmmoEmpty
  (bind _outfortplayerpawn (CallFunction|GetAvatarasValidFortPlayerPawn)
    (:ValidFortPlayerPawn
      (Utilities|IsValid (Weapon|GetCurrentWeapon _outfortplayerpawn)
        (:Is%20Valid
          (return (>= (- (FortWeapon|GetBulletsPerClip (Weapon|GetCurrentWeapon _outfortplayerpawn)) (FortWeapon|GetMagazineAmmoCount (Weapon|GetCurrentWeapon _outfortplayerpawn))) (FortWeapon|GetBulletsPerClip (Weapon|GetCurrentWeapon _outfortplayerpawn)))))))
    (:AvatarCastFailed
      (return false))))
